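# Post-process material that renders a screen collapse effect with a flickering light flare, centered on a camera-relative offset projected to screen space.
# MPC_PostProcessControl's ProgressCollapse drives radially masked, distance-weighted warping of scene-color UVs and an edge-darkening mask.
# Spikeball and cross flare textures, scaled, flickered by layered sines and brightness-clamped per shading path, are added on top.
# RestoreFrontend fades out darkening and flare, leaving the warped scene color.
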Material /Game/Packages/Fortnite_UniqueMaterials_1/PostProcess/PPM_HLODWPOpost2
graph {
 properties {blendableLocation=BL_SceneColorAfterDOF, materialDomain=MD_PostProcess, pixelDepthOffsetMode=PDOM_Legacy, usedShadingModels=Default Lit}
 nodes [
  n1 = MaterialFunctionCall(World Position=Add(A=CameraPositionWS(), B=VectorParameter() {parameterName=CenterOffset, defaultValue=<rgba 3500 -100000 -2000 1>}.RGB)) {materialFunction=&TransformToClipSpace}
  n2 = Multiply(A=@n1.Clip Space XY) {constB=-1}
  Add(A=@n2, B=TextureCoordinate())
  Add(A=@n2, B=TextureCoordinate())
  Add(A=@n2, B=TextureCoordinate())
  AppendVector(A=Constant(), B=ScalarParameter() {defaultValue=-0.07, parameterName=Offset})
  n3 = Frac(None=Multiply(A=Time(), B=ScalarParameter() {defaultValue=0.5, parameterName=ColapseSpeed}))
  Divide(A=@n3)
  If()
  n4 = TextureCoordinate()
  n5 = ScreenPosition()
  n6 = Subtract(A=@n1.Screen aligned 0-1 UV, B=@n5.ViewportUV)
  n7 = Distance(A=@n5.ViewportUV, B=@n1.Screen aligned 0-1 UV)
  n8 = Divide(
   A=Power(Base=OneMinus(None=@n7), Exp=ScalarParameter() {defaultValue=2, parameterName=DistanceFalloff})
   B=ScalarParameter() {defaultValue=3, parameterName=Distance2}
  )
  n9 = Constant()
  n10 = Saturate(None=Power(
    Base=CollectionParameter() {collection=&/MPC_PostProcessControl, parameterName=ProgressCollapse}
    Exp=ScalarParameter() {defaultValue=3, parameterName=ExponentialTime}
   ))
  n11 = LinearInterpolate(A=Constant(), B=ScalarParameter() {defaultValue=-30, parameterName=DistanceTest2}, Alpha=@n10)
  n12 = Constant2Vector() {g=0.5, r=0.5}
  n13 = SphereMask(A=@n4, B=@n12, Radius=ScalarParameter() {defaultValue=0.5, parameterName=Radius}, Hardness=ScalarParameter() {parameterName=Hardness})
  n14 = ScalarParameter() {defaultValue=8, parameterName=ScreenEdgeDelay}
  n15 = LinearInterpolate(
   A=Power(Base=@n13, Exp=ScalarParameter() {defaultValue=1.3679, parameterName=FallofRadiusAExp})
   B=Constant() {r=1}
   Alpha=Saturate(None=Power(Base=@n10, Exp=@n14))
  )
  n16 = Add(A=@n4, B=Multiply(A=Multiply(A=@n6, B=@n8), B=LinearInterpolate(A=@n9, B=@n11, Alpha=@n15)))
  n17 = Constant2Vector() {g=0.5, r=0.5}
  n18 = ScalarParameter() {defaultValue=0.5, parameterName=Radius}
  n19 = ScalarParameter() {parameterName=Hardness}
  n20 = OneMinus(None=@n10)
  n21 = ScalarParameter() {defaultValue=4.2701, parameterName=ScreenEdgeDelay2}
  n22 = LinearInterpolate(
   A=Power(Base=SphereMask(A=@n16, B=@n17, Radius=@n18, Hardness=@n19), Exp=ScalarParameter() {defaultValue=1.5705, parameterName=FallofRadiusAExp2})
   B=Constant() {r=1}
   Alpha=Saturate(None=Power(Base=@n20, Exp=@n21))
  )
  n23 = ComponentMask(None=SceneTexture(UVs=@n16) {sceneTextureId=PPI_PostProcessInput0}.Color) {b=true, g=true, r=true}
  n24 = Constant3Vector() {constant=<rgba 2.67 4.6843 5 1>}
  n25 = Subtract(A=Subtract(A=ScreenPosition().ViewportUV, B=@n1.Screen aligned 0-1 UV), B=Constant2Vector() {g=-0.5, r=-0.5})
  n26 = Constant2Vector() {g=1, r=1}
  n27 = MaterialFunctionCall(Float4=CollectionParameter() {collection=&MPC_PostProcessControl, parameterName=Flare_Scale}) {materialFunction=&BreakOutFloat4Components}
  n28 = Constant() {r=0.01}
  n29 = MaterialFunctionCall(
   X=Clamp(None=@n27.R, Min=@n28) {clampMode=CMODE_ClampMin, minDefault=0.01}
   Y=Clamp(None=@n27.G, Min=@n28) {clampMode=CMODE_ClampMin, minDefault=0.01}
   Z=Clamp(None=@n27.B, Min=@n28) {clampMode=CMODE_ClampMin, minDefault=0.01}
   A=Clamp(None=@n27.A, Min=@n28) {clampMode=CMODE_ClampMin, minDefault=0.01}
  ) {materialFunction=&MakeFloat4}
  n30 = ComponentMask(None=@n29.Result) {g=true, r=true}
  n31 = SphereMask(
   A=MaterialFunctionCall(UVs=@n25, Texture Scale=Multiply(A=@n26, B=@n30)) {materialFunction=&ScaleUVsByCenter}.UVs
   B=Constant() {r=0.5}
  ) {attenuationRadius=0.6, hardnessPercent=0}
  n32 = Multiply(A=Divide(A=Power(Base=@n31) {constExponent=4}) {constB=25}, B=Constant3Vector() {constant=<rgba 0.578 0.7602 1 0>})
  n33 = Multiply(A=Constant2Vector() {g=0.6, r=0.3}, B=@n30)
  n34 = Multiply(A=@n33, B=Constant2Vector() {g=0.8, r=3}) {constB=0.8}
  n35 = Time()
  n36 = Multiply(A=Add(A=@n35) {constB=5.7}) {constB=7}
  n37 = Sine(None=@n36) {period=0.155}
  n38 = Sine(None=@n36) {period=0.2}
  n39 = Multiply(A=Sine(None=@n36) {period=0.075}, B=Multiply(A=Multiply(A=@n37, B=@n38), B=Sine(None=@n36) {period=0.43}))
  n40 = LinearInterpolate(Alpha=@n39) {constA=0.95}
  n41 = Rotator(Coordinate=MaterialFunctionCall(UVs=@n25, Texture Scale=Multiply(A=@n34, B=@n40)) {materialFunction=&ScaleUVsByCenter}.UVs) {speed=-0.15}
  n42 = MaterialFunctionCall(UVs=@n25, Texture Scale=Multiply(A=@n33, B=@n40)) {materialFunction=&ScaleUVsByCenter}
  n43 = Add(
   A=Divide(A=TextureSample(UVs=@n41) {samplerSource=SSM_Clamp_WorldGroupSettings, texture=/Game/Effects/Fort_Effects/Textures/Flares/T_Flare_Spikeball}.RGB) {constB=3}
   B=TextureSample(UVs=Rotator(Coordinate=@n42.UVs) {speed=0.11}) {samplerSource=SSM_Clamp_WorldGroupSettings, texture=&T_Flare_Spikeball}.RGB
  )
  n44 = Multiply(A=@n24, B=Add(A=Clamp(None=@n32), B=Clamp(None=@n43)))
  n45 = ComponentMask(None=@n29.Result) {a=true, b=true}
  n46 = Constant2Vector() {g=0.9, r=1.2}
  n47 = MaterialFunctionCall(UVs=@n25, Texture Scale=Multiply(A=Multiply(A=@n45, B=@n46), B=@n40)) {materialFunction=&ScaleUVsByCenter}
  n48 = TextureSample(UVs=@n47.UVs) {
   samplerSource=SSM_Clamp_WorldGroupSettings
   samplerType=SAMPLERTYPE_Grayscale
   texture=/Game/Effects/Fort_Effects/Textures/Flares/T_Cross_Flare
  }
  n49 = Clamp(None=Divide(A=@n48.RGB) {constB=9})
  n50 = CollectionParameter() {collection=&MPC_PostProcessControl, parameterName=Flare_Flicker}
  n51 = MaterialFunctionCall(Hue Shift Percentage=Constant() {r=0.07}, Texture=Multiply(A=Multiply(A=@n24, B=@n49), B=LinearInterpolate(A=@n50, Alpha=@n39) {constA=0.7})) {materialFunction=&HueShift}
  n52 = Multiply(A=@n35) {constB=5}
  n53 = Multiply(A=Sine(None=@n52) {period=0.155}, B=Sine(None=@n52) {period=0.2})
  n54 = Sine(None=@n52) {period=0.43}
  n55 = LinearInterpolate(A=@n50, Alpha=Multiply(A=Sine(None=@n52) {period=0.075}, B=Multiply(A=@n53, B=@n54))) {constA=0.7}
  n56 = Multiply(
   A=CollectionParameter() {collection=&MPC_PostProcessControl, parameterName=Flare_Brightness}
   B=Multiply(A=Add(A=@n44, B=@n51.Result), B=Clamp(None=@n55) {maxDefault=3})
  )
  n57 = ShadingPathSwitch(Default=Constant() {r=10}, Mobile=ScalarParameter() {defaultValue=4, parameterName=MaxBirghtness}) {desc=Migrated from FeatureLevelSwitch}
  LinearInterpolate(
   A=Add(A=Multiply(A=@n22, B=@n23), B=Clamp(None=@n56, Max=@n57))
   B=@n23
   Alpha=CollectionParameter() {collection=&MPC_PostProcessControl, parameterName=RestoreFrontend}
  )
  OneMinus(None=@n13)
  n58 = ScalarParameter() {defaultValue=1, parameterName=PauseAdd}
  Reroute(None=Saturate(None=Multiply(A=@n3, B=@n58) {constB=2}))
  Rotator(Coordinate=@n47.UVs) {speed=0.08}
  n59 = Distance(A=@n4, B=@n12)
  n60 = ScalarParameter() {defaultValue=-0.15, parameterName=FallofRadiusAOffset}
  Saturate(None=Divide(A=Add(A=@n59, B=@n60), B=ScalarParameter() {defaultValue=0.5, parameterName=FallofRadiusA}))
  ScalarParameter() {defaultValue=0.5, parameterName=Progress}
 ]
}
