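# Post-process material applied after depth of field that grades the scene for the Sideways bubble, as the split-screen player 2 variant.
# World position relative to the P2 bubble location and scale yields a masked gradient (Z by default) that drives a curve-atlas color and brightness.
# A sphere mask around the shared PlayerPosition parameter, not a P2-specific one, restores original color.
# A depth-based two-color fog blends over the result.
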Material /Game/Packages/Fortnite_UniqueMaterials_1/PostProcess/PPM_Sideways_SplitScreen_P2
graph {
 properties {blendableLocation=BL_SceneColorAfterDOF, materialDomain=MD_PostProcess, pixelDepthOffsetMode=PDOM_Legacy, usedShadingModels=Default Lit}
 nodes [
  n1 = SceneTexture() {sceneTextureId=PPI_PostProcessInput0}
  n2 = CollectionParameter() {
   collection=&FortniteMaterialParameters
   parameterName=SidewaysBubble_Location_LocalP2
  }
  n3 = Subtract(A=WorldPosition().XYZ, B=ComponentMask(None=@n2) {b=true, g=true, r=true})
  n4 = CollectionParameter() {
   collection=&FortniteMaterialParameters
   parameterName=SidewaysBubble_Scale_LocalP2
  }
  n5 = Clamp(None=ChannelMaskParameter(None=Divide(A=@n3, B=@n4)) {maskChannel=Blue, defaultValue=<rgba 0 0 1 0>}.R)
  n6 = LinearInterpolate(A=Constant4Vector() {constant=<rgba 0.0408 0.0228 0.2083 0>}.RGBA, B=Constant4Vector() {constant=<rgba 1 0.2836 0.0112 0>}.RGBA, Alpha=@n5)
  Add(A=@n1.Color, B=@n6)
  n7 = CurveAtlasRowParameter(CurveTime=@n5) {
   atlas=/Game/Athena/Apollo/Environments/Landscape/Materials/Corruption/CorruptionCurveAtlas
   curve=/Game/Athena/Apollo/Environments/Landscape/Materials/Corruption/sideways_PPCurve
   defaultValue=7
   parameterName=PP_GradientColorCurve
  }
  n8 = Multiply(
   A=AppendVector(A=@n7, B=@n7)
   B=LinearInterpolate(
    A=ScalarParameter() {defaultValue=1, parameterName=MinBrightness}
    B=ScalarParameter() {defaultValue=1, parameterName=Max_Brightness}
    Alpha=@n5
   )
  )
  n9 = ScalarParameter() {defaultValue=0.7, parameterName=PP_BlendAmt}
  n10 = WorldPosition()
  n11 = CollectionParameter() {collection=&FortniteMaterialParameters, parameterName=PlayerPosition}
  n12 = ComponentMask(None=@n11) {b=true, g=true, r=true}
  n13 = ScalarParameter() {defaultValue=512, parameterName=PLayerMaskRadius}
  n14 = Constant()
  n15 = Multiply(
   A=VectorParameter() {parameterName=FogColor, defaultValue=<rgba 0.1325 0 0.2604 1>}.RGB
   B=ComponentMask(None=@n1.Color) {b=true, g=true, r=true}
  )
  n16 = VectorParameter() {parameterName=FogColor_02, defaultValue=<rgba 0.1823 0.0499 0.0077 1>}
  n17 = SceneDepth()
  n18 = ScalarParameter() {defaultValue=512, parameterName=FogDepthBias}
  n19 = Clamp(None=Divide(A=Subtract(A=@n17.R, B=@n18), B=ScalarParameter() {defaultValue=4096, parameterName=FogDepth}))
  LinearInterpolate(
   A=LinearInterpolate(A=LinearInterpolate(A=@n1.Color, B=@n8, Alpha=@n9), B=@n1.Color, Alpha=SphereMask(A=@n10.XYZ, B=@n12, Radius=@n13, Hardness=@n14))
   B=AppendVector(A=LinearInterpolate(A=@n15, B=@n16.RGB, Alpha=@n19), B=Constant())
   Alpha=ScalarParameter() {defaultValue=0.5, parameterName=Fog_BlendAmt}
  )
  Multiply(A=@n1.Color, B=@n6)
  Multiply(A=Desaturation(None=@n1.Color, Fraction=Constant() {r=1}), B=@n5)
  ScalarParameter() {defaultValue=5000, parameterName=GradientScale}
  VectorParameter() {parameterName=Z_Offset, defaultValue=<rgba 0 0 0 1>}
 ]
}
